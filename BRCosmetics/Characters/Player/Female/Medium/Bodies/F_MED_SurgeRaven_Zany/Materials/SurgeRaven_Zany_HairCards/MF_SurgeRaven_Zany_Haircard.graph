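# Material function for the SurgeRaven Zany character's hair cards that overrides incoming material attributes with hair-specific shading.
# A packed mask texture drives opacity, blending temporal-AA dithered and plain masks by pixel-depth fade and zeroing them in shadow passes.
# A UV-V gradient brightens base color along the card.
# A card normal map blends with two-sided-corrected vertex normals, alongside tunable roughness, metallic, specular and anisotropy, plus fixed tangent and depth offset.
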MaterialFunction /BRCosmetics/Characters/Player/Female/Medium/Bodies/F_MED_SurgeRaven_Zany/Materials/SurgeRaven_Zany_HairCards/MF_SurgeRaven_Zany_Haircard
graph {nodes [
  n1 = Multiply(A=VertexNormalWS(), B=TwoSidedSign())
  n2 = TextureSample() {texture=&/T_SurgeRaven_Zany_HairCard_N}
  n3 = SetMaterialAttributes!(
   MaterialAttributes=NamedRerouteUsage()
   BaseColor=NamedRerouteUsage()
   OpacityMask=NamedRerouteUsage()
   WorldPositionOffset
   Opacity=LinearInterpolate(A=Constant() {r=1}, Alpha=TwoSidedSign())
   SubsurfaceColor=Constant()
   Roughness=ScalarParameter() {defaultValue=0.5, group=ZanyHairCard, parameterName=MessHairRoughness}
   Normal=MaterialFunctionCall(BaseNormal=Transform(None=@n1) {transformSourceType=TRANSFORMSOURCE_World, transformType=TRANSFORM_Tangent}, AdditionalNormal=@n2.RGB) {materialFunction=&BlendAngleCorrectedNormals}.Result
   Metallic=ScalarParameter() {defaultValue=0.1, group=ZanyHairCard, parameterName=HairCard_Metallic}
   Specular=ScalarParameter() {defaultValue=0.9, group=ZanyHairCard, parameterName=HairCard_Specular}
   Anisotropy=ScalarParameter() {defaultValue=0.3637, group=ZanyHairCard, parameterName=HairCard_Anisotropy}
   PixelDepthOffset=Constant() {r=-0.1}
   Tangent=Constant4Vector() {constant=<rgba 0 1 0 1>}.RGB
  ) {attributeSetTypes [
    #1
    #2
    #3
    #4
    #5
    #6
    #7
    #8
    #9
    #10
    #11
    #12
   ]}
  FunctionOutput(None=@n3)
  n4 = GetMaterialAttributes(None=FunctionInput() {inputName=MA, inputType=FunctionInput_MaterialAttributes}) {attributeGetTypes [#1, #7]}
  NamedRerouteDeclaration(None=@n4.BaseColor) {name=InputBaseColor, nodeColor=<rgba 1 0 0.6328 1>}
  n5 = TextureSampleParameter2D(UVs=TextureCoordinate()) {
   desc=R-MesssHairLayer1 G-MessHairLayLayer2 B -MesshairMask
   group=ZanyHairCard
   parameterName=MessHairMask
   samplerSource=SSM_Wrap_WorldGroupSettings
   samplerType=SAMPLERTYPE_Masks
   texture=&T_BlackMask
  }
  NamedRerouteDeclaration(None=@n5.B) {name=MaskThreshold, nodeColor=<rgba 0.6641 0 1 1>}
  n6 = Max(A=@n5.R, B=@n5.R)
  n7 = Power(Base=@n6, Exp=ScalarParameter() {defaultValue=1.9, group=ZanyHairCard, parameterName=Dither Close}) {constExponent=1.8}
  n8 = Saturate(None=LinearInterpolate(A=Constant() {r=-0.1}, B=Constant() {r=2}, Alpha=NamedRerouteDeclaration(None=@n7) {name=MessHairMaskB_2, nodeColor=<rgba 1 0 0.9141 1>}) {constA=0.5})
  n9 = PixelDepth()
  n10 = Constant() {r=30}
  n11 = LinearInterpolate(Alpha=Saturate(None=Divide(A=@n9.R, B=@n10) {constB=100})) {constA=1, constB=0}
  n12 = LinearInterpolate(
   A=NamedRerouteDeclaration(None=MaterialFunctionCall(Alpha Threshold=@n8, Random=@n11) {materialFunction=&DitherTemporalAA}.Result) {name=DitherOpacitymask, nodeColor=<rgba 0.9922 0 1 1>}
   B=NamedRerouteUsage()
   Alpha=Power(Base=MaterialFunctionCall() {materialFunction=/Game/Global/Materials/Character_Parent_2023/Hair_Parent_2023/MaterialFunctions/MF_PixelDepthFadeIn}.Result) {constExponent=3}
  )
  n13 = Constant()
  n14 = Power(Base=Multiply(A=NamedRerouteUsage()) {constB=3})
  NamedRerouteDeclaration(None=Multiply(A=ShadowReplace(Default=@n12, Shadow=@n13), B=Saturate(None=@n14))) {name=Opacitymask_1, nodeColor=<rgba 0.0078 0 1 1>}
  n15 = Constant() {r=1}
  n16 = Constant() {r=4}
  n17 = ComponentMask(None=TextureCoordinate()) {g=true}
  n18 = Power(Base=Saturate(None=Multiply(A=@n17)), Exp=Constant() {r=3.5186})
  NamedRerouteDeclaration(None=Multiply(A=NamedRerouteUsage(), B=LinearInterpolate(A=@n15, B=@n16, Alpha=@n18))) {name=BaseColor_1, nodeColor=<rgba 0.3125 1 0 1>}
  NamedRerouteDeclaration(None=@n4.MaterialAttributes) {name=InputMA_1, nodeColor=<rgba 1 0 0.2344 1>}
  n19 = NamedRerouteDeclaration(None=Power(Base=@n6, Exp=ScalarParameter() {defaultValue=1.2, group=ZanyHairCard, parameterName=Dither Far}) {constExponent=1.8}) {name=MessHairMaskB_1, nodeColor=<rgba 1 0 0.9141 1>}
  NamedRerouteDeclaration(None=ShadowReplace(Default=Saturate(None=@n19), Shadow=Constant())) {name=BaseOpacitymask, nodeColor=<rgba 0.0078 0 1 1>}
 ]}
